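; Fortnite weapon Blueprint for the player augment selector tool, with placeholder hooks for its cosmetic hologram effects.
; A custom startup event activates the granter lightcone effect, but nothing here calls it; a TODO plans an animation notify trigger.
; BeginPlay and unequip call a shutdown event, gated on cosmetic-event support, but that event is empty.
; The equip and destroyed events are also empty, and construction only calls the parent.
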(event EventBeginPlay
  (Networking|CanExecuteCosmeticEvents
    (:True
      (CallFunction|CustomonUnEquippedTempSoloShutdown))))

(event FortWeapon|EventOnEquippedAndAttached)

(event EventDestroyed)

(event FortWeapon|EventOnUnEquippedAndDetached
  (Networking|CanExecuteCosmeticEvents
    (:True
      (CallFunction|CustomonUnEquippedTempSoloShutdown))))

(event Custom|CustomOnEquippedTempHoloStartup
  ;[0.9034 1 0.0904] TODO: Anim keyframe notify to kick this stuff off
  (Components|Activation|Activate (Variables|Default|GetNS_Aug_Granter_Lightcone) false))

(event Custom|CustomOnUnEquippedTempSoloShutdown)

(fn ConstructionScript
  (|CallParentFunction))
