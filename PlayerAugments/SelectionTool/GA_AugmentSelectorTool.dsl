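; Gameplay ability for the augment selector tool, letting players select or reroll augments.
; On activation it binds to the controller's augment granting component and caches available choices, refreshing them on updates.
; Code-bound input events drive highlight cues, hold-to-reroll audio and rumble, and selection, which pushes a camera mode, plays a montage, then grants after a delay.
; EndAbility grants any pending choice; granting pops the camera after a blend delay, supporting consecutive selections.
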(event Ability|EventActivateAbility
  ;[1 1 1] Activate Ability - Save out variables and roll for the augments if they have not been rolled yet.
  (CallFunction|CachePlayerRefs)
  (Utilities|Array|Clear (Variables|Default|GetCachedChoices))
  (CallFunction|BindGrantingEvents)
  (CallFunction|UpdateChoices)
  (Variables|Default|SetChoice%20Index -1))

(event Ability|EventFailedToActivatePassiveAbility
  (params (ActorInfo :type Gameplay%20Ability%20Actor%20Info%20Structure))
  ;[1 1 1] Need this for Passive Abilities
  (|CallParentFunction))

(event EventOnSelectedChoiceIndex
  (params (ChoiceIndex :type Integer))
  ;[1 1 1] On Selected Choice Index - this event comes from code bound input.
  (Variables|Default|SetChoice%20Index ChoiceIndex)
  (Camera|PushCameraMode "/PlayerAugments/SelectionTool/Athena_PlayerCameraModeAugmentSelector_v2.Athena_PlayerCameraModeAugmentSelector_v2_C")
  (GameplayCue|ExecuteGameplayCue%28Burst%29Local (Ability|GetAvatarActorfromActorInfo) (Variables|Default|GetGC_Select) (Ability|GameplayCue|MakeGameplayCueParameters :NormalizedMagnitude 0.0 :RawMagnitude 2e+02 :Location (Transformation|GetActorLocation (Ability|GetAvatarActorfromActorInfo)) :Normal "0,0,0" :GameplayEffectLevel 1 :AbilityLevel 1 :bReplicateLocationWhenUsingMinimalRepProxy false))
  (Ability|Tasks|PlayMontageandWaitforTrigger "None" (Variables|Default|GetConfirmMontage)) ; looks like this might send it to server, so perhaps\nneed to actually only call this is if roll is "autonomous proxy"
  (Utilities|FlowControl|Delay 1.25)
  (CallFunction|ClientGrantAugmentChoice (Variables|Default|GetChoice%20Index)))

(event Custom|NotifyAugmentGranted
  (params (AugDef :type Fort%20Item%20Definition%20Object%20Reference)))

(event Custom|ClientGrantAugmentChoice
  (params (ChoiceIndex :type Integer))
  ;[1 1 1] Client grant choice (NOTE: this
  (Variables|Default|SetChoice%20Index -1)
  (bind _outfortplayerpawn (CallFunction|GetAvatarasValidFortPlayerPawn)
    (:ValidFortPlayerPawn
      (bind _returnvalue (Class|FortControllerComponentAugmentGrantingSystem|GetAugmentGrantingComponentfromController (Pawn|GetController _outfortplayerpawn)))
      (Utilities|IsValid _returnvalue
        (:Is%20Valid
          (Class|FortControllerComponentAugmentGrantingSystem|ClientSelectAvailableAugmenttoGrant _returnvalue (Utilities|Array|Get(acopy) (Variables|Default|GetCachedChoices) ChoiceIndex))
          ;;[1 1 1] FORT-586048: Pop camera, which takes camera back to Augment Device camera, as we now allow this GA to continue if multiple augment selections are available. Delay matches camera blend time. If its single selection, GA ends before this delay ends, so camera is normal blend into new weapon
          (Utilities|FlowControl|Delay 0.45)
          (Camera|PopCameraMode))))))

(event EventOnHighlightSelectionUpdated
  (params (NewHighlightedChoiceIndex :type Integer))
  ;[1 1 1] Highlight Selection Updated - This event comes from code bound input
  (GameplayCue|ExecuteGameplayCue%28Burst%29Local (Ability|GetAvatarActorfromActorInfo) (Variables|Default|GetGC_Highlight) (Ability|GameplayCue|MakeGameplayCueParameters :NormalizedMagnitude 0.0 :RawMagnitude 1e+02 :Location (Transformation|GetActorLocation (Ability|GetAvatarActorfromActorInfo)) :Normal "0,0,0" :GameplayEffectLevel 1 :AbilityLevel 1 :bReplicateLocationWhenUsingMinimalRepProxy false)))

(event Ability|EventOnEndAbility
  (params (bWasCancelled :type Boolean))
  (bind _choice_index (Variables|Default|GetChoice%20Index))
  ;[1 1 1] End Ability - If you selected an augment grant it to the player
  (|CallParentFunction bWasCancelled)
  (if (>= _choice_index 0)
    (CallFunction|ClientGrantAugmentChoice _choice_index)))

(event Custom|BindGrantingEvents
  ;[1 1 1] Bind Granting Events
  (bind _outfortplayerpawn (CallFunction|GetAvatarasValidFortPlayerPawn)
    (:ValidFortPlayerPawn
      (bind _returnvalue (Class|FortControllerComponentAugmentGrantingSystem|GetAugmentGrantingComponentfromController (Pawn|GetController _outfortplayerpawn)))
      (EventDispatchers|AssignOnLocalReroll _returnvalue (delegate Custom|OnLocalReroll_Event))
      (EventDispatchers|AssignOnAvailableAugmentsforGrantingUpdated _returnvalue (delegate Custom|OnAvailableAugmentsForGrantingUpdated_Event)))))

(event Custom|OnAvailableAugmentsForGrantingUpdated_Event
  (params (GrantingComponent :type Fort%20Controller%20Component%20Augment%20Granting%20System%20Object%20Reference))
  ;[1 1 1] On Reroll Updated Choices
  (CallFunction|UpdateChoices)
  (CustomVerb|BroadcastCustomVerbMessage (CustomVerb|MakeCustomVerbMessage (Variables|Default|GetPlayerRef) "(TagName=\"CustomVerb.AugmentActivated\")")))

(event Custom|OnLocalReroll_Event
  ;[1 1 1] Reroll animation
  (Ability|Tasks|PlayMontageandWaitforTrigger "None" (Variables|Animation|GetRerollMontage)) ; looks liek this might send it to server, so perhaps\nneed to actually only call this is if roll is "autonomous proxy"
  (Audio|SpawnSoundatLocation (Variables|Default|GetHoldLoopFinishSound) (Transformation|GetActorLocation (Variables|Default|GetPlayerRef)))
  (Ability|ExecuteGameplayCueOnOwner (Variables|Default|GetGC_Reroll))
  (bind _outfortplayerpawn (CallFunction|GetAvatarasValidFortPlayerPawn)
    (:ValidFortPlayerPawn
      (Quest|SendComplexCustomStatEvent (Quests|GetQuestManager (Fortnite|GetFortPlayerControllerfromActor _outfortplayerpawn) "Athena") 0 "(GameplayTags=((TagName=\"Athena.Quests.RerolledAugments\")))" "(GameplayTags=)" 1))))

(event Custom|UpdateChoices
  ;[1 1 1] Update Choices
  (Utilities|Array|Clear (Variables|Default|GetCachedChoices))
  (bind _outfortplayerpawn (CallFunction|GetAvatarasValidFortPlayerPawn)
    (:ValidFortPlayerPawn
      (bind _returnvalue (Class|FortControllerComponentAugmentGrantingSystem|GetAugmentGrantingComponentfromController (Pawn|GetController _outfortplayerpawn)))
      (Utilities|IsValid _returnvalue
        (:Is%20Valid
          (for _array_element (Class|FortControllerComponentAugmentGrantingSystem|GetAvailableAugmentstoGrant _returnvalue)
            (Utilities|Array|Add (Variables|Default|GetCachedChoices) (Utilities|Struct|BreakPlayerAugmentChoiceDatum _array_element))))
        (:Is%20Not%20Valid
          (Development|PrintString "No Granter Comp"))))))

(event EventOnRerollPressedAndCannotAfford
  (Audio|SpawnSound2D (Variables|Default|GetCanNotAffordSound))
  (ForceFeedback|SpawnForceFeedbackAttached "/Game/UI/InGame/HUD/AthenaRadial/FF_RadialItemSelection.FF_RadialItemSelection" (Variables|/Script/Engine.Actor|GetRootComponent (Variables|Default|GetPlayerRef))))

(event EventOnHoldRerollInputChanged
  (params (bIsPressed :type Boolean) (bCompletedHold :type Boolean))
  (if bIsPressed
    (bind _returnvalue (Audio|SpawnSoundatLocation (Variables|Default|GetHoldLoopSound) (Transformation|GetActorLocation (Variables|Default|GetPlayerRef))))
    (bind _output_get (Variables|Default|SetHoldAudioComponent _returnvalue))
    (Audio|Components|Audio|FadeIn _output_get 0.1 1.0 0.0 "Linear" :!class /Script/Engine.AudioComponent)
    (ForceFeedback|SpawnForceFeedbackAttached "/PlayerAugments/SelectionTool/Assets/ForceFeedback/FF_Augment_RerollHold.FF_Augment_RerollHold" (Variables|/Script/Engine.Actor|GetRootComponent (Variables|Default|GetPlayerRef)))
    (else
      (bind _holdaudiocomponent (Variables|Default|GetHoldAudioComponent)
        (:then
          (Audio|Components|Audio|FadeOut _holdaudiocomponent 0.1 0.0 "Linear" :!class /Script/Engine.AudioComponent))
        (:else)))))

(fn CachePlayerRefs
  (bind _outfortplayerpawn (CallFunction|GetAvatarasValidFortPlayerPawn)
    (:ValidFortPlayerPawn
      (Utilities|IsValid (Fortnite|GetFortPlayerControllerfromActor _outfortplayerpawn)
        (:Is%20Valid
          (Variables|Default|SetPlayerRef _outfortplayerpawn)
          (Variables|Default|SetPlayerControllerRef (Fortnite|GetFortPlayerControllerfromActor _outfortplayerpawn)))))))
